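# Niagara one-shot burst effect for the StylusFluff emote's equip moment, built like a stylized swing-impact hit.
# Flare and corona, flash streaks, speed lines, sparks, toon smoke and curl-noise glow drops burst around a shared System.OffSetAfterMotionBlur time.
# System values drive scale, time-of-day tint and drop direction, and SpeedLine splits its looks by spawn group through a visibility tag.
# The light-casting flare and both comic-line emitters are disabled.
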NiagaraSystem /BRCosmetics/Animation/Game/MainPlayer/Emotes/StylusFluff/VFX/NS_Emote_StylusFluff_Equip2
dependencies {
 usedDataInterfaces [
  &NiagaraDataInterfaceCurve
  &NiagaraDataInterfaceColorCurve
  &NiagaraDataInterfaceArrayFloat
  &NiagaraDataInterfaceVector2DCurve
  &NiagaraDataInterfaceArrayInt32
  &NiagaraDataInterfaceCurlNoise
  &NiagaraDataInterfaceArrayFloat3
 ]
 usedDynamicInputs [&ReturnNormalizedExecIndex, &FN_FrontEndFloatModifier]
 usedModules [
  &self:SystemSpawnScript.SystemScriptSource.SystemScriptGraph.NiagaraNodeAssignment_2.SetVariables_#1
  &self:SystemSpawnScript.SystemScriptSource.SystemScriptGraph.NiagaraNodeAssignment_1.SetVariables_#2
  &self:SystemSpawnScript.SystemScriptSource.SystemScriptGraph.NiagaraNodeAssignment_4.SetVariables_#3
  &self:SystemSpawnScript.SystemScriptSource.SystemScriptGraph.NiagaraNodeAssignment_5.SetVariables_#4
  &SystemState
 ]
}
emitters {
 ComicLinesAxeMotionBlur {
  bEnabled=false
  emitterUpdateScript [
   &EmitterState(
    Enable Distance Culling=true
    Max Distance Response=::KillImmediately
    Module.bApplySpawnCountScaleByDistance=true
    Module.bUseMaxDistance=true
    Module.bUseMinDistance=false
    Scalability Mode=::NewEnumerator1
    Scale Spawn Count=true
    Spawn Count Scale By Distance=DataInterface() {propertyValues {Curve {keys [<at 0.3676 0.9657>,<at 0.9962 0.4982>]}}}
   )
   &SpawnBurst_Instantaneous(Spawn Count=10, Use Spawn Probability=false, UseLoopCountLimit=false)
  ]
  particleSpawnScript [
   &PartitionParticles(Particles Per Partition=9, Partition Mode:ENiagara_PartitionMode=::NewEnumerator1)
   &InitializeParticle(
    Color=<rgba 0 0 0 1>
    Lifetime Max=0.3
    Lifetime Min=0.25
    Lifetime Mode=::NewEnumerator1
    Position Offset=<0 0 -2.4002>
    Position Offset Coordinate Space=::Local
    Sprite Size Mode=::NewEnumerator1
    Uniform Sprite Size Max=3
    Uniform Sprite Size Min=2
    UsePositionOffset=true
    Write Material Random=true
   )
   &ShapeLocation(
    Box Size=<1 50 50>
    MaskSpawnGroup=false
    Offset Coordinate Space=::Local
    Offset Mode=::NewEnumerator0
    Override Seed=false
    Shape Primitive=::NewEnumerator2
    Surface Only=false
    UseOwnerScale=true
   )
   &AddVelocity(Velocity=$System.DirectionVector, Velocity Speed Scale=2000)
  ]
  particleUpdateScript [
   &ParticleState()
   &GravityForce(Gravity=<0 0 -50>) {enabled=false}
   &Drag(Drag=&UniformRangedFloat(Minimum=10, Maximum=15), Use Linear Drag=true, Use Rotational Drag=false)
   &CurlNoiseForce(Noise Strength=&RandomRangeFloat(Minimum=200, Maximum=400), Pan Noise=true)
   &SolveForcesAndVelocity(Clamp Velocity=false, Limit Acceleration=false)
   &ScaleSpriteSize(Scale Factor=&Vector2DFromCurve(
     Vector2Curve=DataInterface() {propertyValues {
       XCurve {keys [<at 0 1 RCIM_Cubic>, {arriveTangent=-2.8959, interpMode=RCIM_Cubic, leaveTangent=-2.8959, tangentMode=RCTM_User, time=1, value=0}]}
       YCurve {keys [<at 0 1 RCIM_Cubic>, {arriveTangent=-2.8959, interpMode=RCIM_Cubic, leaveTangent=-2.8959, tangentMode=RCTM_User, time=1, value=0}]}
      }}
     ScaleCurve=<1 1>
    ))
   &ScaleSpriteSizeBySpeed(Max Scale Factor=<0.5 15>, Min Scale Factor=<0.8 1.2>)
  ]
  rendererClasses [&NiagaraSpriteRendererProperties]
  renderers [{rendererClass=&NiagaraSpriteRendererProperties}]
 }
 ComicLinesImpact {
  bEnabled=false
  emitterUpdateScript [
   &EmitterState(
    Enable Distance Culling=true
    Max Distance Response=::KillImmediately
    Module.bApplySpawnCountScaleByDistance=true
    Module.bUseMaxDistance=true
    Module.bUseMinDistance=false
    Scalability Mode=::NewEnumerator1
    Scale Spawn Count=true
    Spawn Count Scale By Distance=DataInterface() {propertyValues {Curve {keys [<at 0.3676 0.9657>,<at 0.9962 0.4982>]}}}
   )
   &SpawnBurst_Instantaneous(Spawn Count=10, Use Spawn Probability=false, UseLoopCountLimit=false)
  ]
  particleSpawnScript [
   &PartitionParticles(Particles Per Partition=9, Partition Mode:ENiagara_PartitionMode=::NewEnumerator1)
   &InitializeParticle(
    Color=<rgba 0 0 0 1>
    Lifetime Max=0.3
    Lifetime Min=0.25
    Lifetime Mode=::NewEnumerator1
    Position Offset=<0 0 -2.4002>
    Position Offset Coordinate Space=::Local
    Sprite Size Mode=::NewEnumerator1
    Uniform Sprite Size Max=3
    Uniform Sprite Size Min=2
    UsePositionOffset=true
    Write Material Random=true
   )
   &ShapeLocation(
    Box Size=<1 50 50>
    MaskSpawnGroup=false
    Offset Coordinate Space=::Local
    Offset Mode=::NewEnumerator0
    Override Seed=false
    Shape Primitive=::NewEnumerator2
    Surface Only=false
    UseOwnerScale=true
   )
   &AddVelocity(Velocity=$System.DirectionVector, Velocity Speed Scale=2000)
  ]
  particleUpdateScript [
   &ParticleState()
   &GravityForce(Gravity=<0 0 -50>)
   &Drag(Drag=&UniformRangedFloat(Minimum=10, Maximum=15), Use Linear Drag=true, Use Rotational Drag=false)
   &CurlNoiseForce(Noise Strength=&RandomRangeFloat(Minimum=200, Maximum=400), Pan Noise=true)
   &SolveForcesAndVelocity(Clamp Velocity=false, Limit Acceleration=false)
   &ScaleSpriteSize(Scale Factor=&Vector2DFromCurve(
     Vector2Curve=DataInterface() {propertyValues {
       XCurve {keys [<at 0 1 RCIM_Cubic>, {arriveTangent=-2.8959, interpMode=RCIM_Cubic, leaveTangent=-2.8959, tangentMode=RCTM_User, time=1, value=0}]}
       YCurve {keys [<at 0 1 RCIM_Cubic>, {arriveTangent=-2.8959, interpMode=RCIM_Cubic, leaveTangent=-2.8959, tangentMode=RCTM_User, time=1, value=0}]}
      }}
     ScaleCurve=<1 1>
    ))
   &ScaleSpriteSizeBySpeed(Max Scale Factor=<0.5 15>, Min Scale Factor=<0.8 1.2>)
  ]
  rendererClasses [&NiagaraSpriteRendererProperties]
  renderers [{rendererClass=&NiagaraSpriteRendererProperties}]
 }
 Flash_Flare {
  bEnabled=false
  emitterUpdateScript [
   &EmitterState(Life Cycle Mode=::NewEnumerator1, Loop Behavior=::NewEnumerator1, Loop Duration=1, UseLoopDelay=false)
   &SpawnBurst_Instantaneous(Spawn Count=1, Spawn Time=$System.OffSetAfterMotionBlur, Use Spawn Probability=false)
  ]
  particleSpawnScript [
   /Niagara/Modules/Spawn/Initialization/InitializeParticle(
    Color=&LerpLinearColors(StartColor=<rgba 1 0.3938 0 1>, EndColor=$System.TODColor, LerpFactor=0.4)
    Lifetime=0.135
    Mesh Scale=<1 1 0.25>
    Module.Write Color=true
    Module.Write Lifetime=true
    Module.Write Mass=true
    Module.Write Material Random=true
    Module.Write Position=false
    Module.Write Scale=true
    Module.Write SpriteRotation=false
    Module.Write SpriteSize=true
    Position=&SimulationPosition()
    Sprite Size=<200 70>
   )
   /SetVariables(Particles.CoronaSpriteSize=&Vector2DFromFloat(Value=&Multiply_Float(A=90, B=$System.scaleFX))) {isSetParameters=true}
   &CameraOffset(Camera Offset Amount=50)
   &DynamicMaterialParameters(
    Index 0 Param 1=0.2
    Index 0 Param 2=1
    Module.Param0WriteEnabled=true
    Module.Param1WriteEnabled=true
    Module.Param2WriteEnabled=true
    Module.Param3WriteEnabled=true
   )
  ]
  particleUpdateScript [
   &ParticleState()
   &SolveForcesAndVelocity(Clamp Velocity=false, Limit Acceleration=false)
   &ScaleColor(
    Scale Alpha=&Lerp_Float(A=1, B=0, Alpha=$Particles.NormalizedAge)
    Scale RGB=&VectorFromFloat(Value=1)
    ScaleA=true
    ScaleRGB=true
   )
   &FN_Light_v2(Light Exponent=3, Light Radius=150, LightEnabled=true, Scale Color=1)
   /SetVariables(
    Particles.CoronaDY:Vector4f=&MakeVector4(
     X=&FloatFromCurve(FloatCurve=DataInterface() {propertyValues {Curve {keys [<at 0 1 RCIM_Cubic>,<at 0.8 0 RCIM_Cubic>]}}})
     Y=2
     Z=1
     W=1
    )
    Particles.DY0:Vector4f=&MakeVector4(
     X=&FloatFromCurve(FloatCurve=DataInterface() {propertyValues {Curve {keys [{arriveTangent=1, time=0.56, value=0}, <at 1 1>]}}})
     Y=0
     Z=0
     W=0
    )
   ) {isSetParameters=true}
   &ScaleMeshSize(Scale Factor=&VectorFromFloat(Value=&FloatFromCurve(FloatCurve=DataInterface() {propertyValues {Curve {keys [<at 0 0.25>,<at 1 1>]}}})))
   /SetVariables(Particles.CoronaSpriteColor:LinearColor=&MakeLinearColorFromVectorAndFloat(
     "Vector (RGB)"=&MakeVectorFromLinearColorRGB(Color=<rgba 0.5521 0.5521 0.5521 1>)
     "Float (Alpha)"=&Multiply_Float(
      A=&FloatFromCurve(FloatCurve=DataInterface() {propertyValues {Curve {keys [<at 0.35 1.5 RCIM_Cubic>,<at 0.95 0>]}}})
      B=3.9
     )
    )) {isSetParameters=true}
  ]
  rendererClasses [&NiagaraSpriteRendererProperties, &NiagaraLightRendererProperties]
  renderers [
   {rendererClass=&NiagaraSpriteRendererProperties}
   {rendererClass=&NiagaraLightRendererProperties, rendererIndex=1}
  ]
 }
 GlowDrops {
  emitterUpdateScript [
   &EmitterState(
    Enable Distance Culling=true
    Max Distance Response=::KillImmediately
    Module.bApplySpawnCountScaleByDistance=true
    Module.bUseMaxDistance=true
    Module.bUseMinDistance=false
    Scalability Mode=::NewEnumerator1
    Scale Spawn Count=true
    Spawn Count Scale By Distance=DataInterface() {propertyValues {Curve {keys [<at 0.3676 0.9657>,<at 0.9962 0.4982>]}}}
   )
   &SpawnBurst_Instantaneous(
    Spawn Count=8
    Spawn Time=&Add_Float(A=$System.OffSetAfterMotionBlur, B=0.04)
    Use Spawn Probability=false
    UseLoopCountLimit=false
   )
  ]
  particleSpawnScript [
   &PartitionParticles(Particles Per Partition=9, Partition Mode:ENiagara_PartitionMode=::NewEnumerator1)
   /Niagara/Modules/Spawn/Initialization/V2/InitializeParticle(
    Color=&LerpLinearColors(StartColor=<rgba 0.4428 1 0.8068 1>, EndColor=$System.TODColor, LerpFactor=0.7)
    Lifetime Max=0.5
    Lifetime Min=0.25
    Lifetime Mode=::NewEnumerator1
    Sprite Size Mode=::NewEnumerator1
    Uniform Sprite Size Max=2.5
    Uniform Sprite Size Min=1.5
    UsePositionOffset=false
    Write Material Random=true
   )
   &SphereLocation(
    MaskSpawnGroup=false
    Override Seed=false
    Sphere Radius=5
    Surface Only=false
    UseNonUniformScale=false
    UseOffset=false
    UseSphereOrientation=false
   ) {module=SphereLocation001}
   &AddVelocityFromPoint(Module.Use Falloff Distance=false, Velocity Strength=&UniformRangedFloat(Minimum=350, Maximum=850))
   &AddVelocity(Velocity=&Multiply_Vector(
     A=<0 600 0>
     B=&MakeCustomVectorFromBool(
      Bool=$System.LastSwing
      True Vector=&SelectVectorFromArray(
       Array Sampling Mode=::NewEnumerator0
       Direct Array Index=?
       Vector Selection Array=DataInterface() {propertyValues {FloatData [<1 1 1>, <1 -1 1>]}}
       Random Seed=0
      )
      False Vector=$System.DirectionVector
     )
    ))
  ]
  particleUpdateScript [
   &ParticleState()
   &GravityForce(Gravity=<0 0 -50>)
   &Drag(Drag=&UniformRangedFloat(Minimum=10, Maximum=15), Use Linear Drag=true, Use Rotational Drag=false)
   &CurlNoiseForce(Noise Strength=&RandomRangeFloat(Minimum=200, Maximum=400), Pan Noise=true)
   &SolveForcesAndVelocity(Clamp Velocity=false, Limit Acceleration=false)
   &ScaleSpriteSize(Scale Factor=&Vector2DFromCurve(
     Vector2Curve=DataInterface() {propertyValues {
       XCurve {keys [<at 0 1 RCIM_Cubic>, {arriveTangent=-2.8959, interpMode=RCIM_Cubic, leaveTangent=-2.8959, tangentMode=RCTM_User, time=1, value=0}]}
       YCurve {keys [<at 0 1 RCIM_Cubic>, {arriveTangent=-2.8959, interpMode=RCIM_Cubic, leaveTangent=-2.8959, tangentMode=RCTM_User, time=1, value=0}]}
      }}
     ScaleCurve=<1 1>
    ))
   &ScaleSpriteSizeBySpeed(Max Scale Factor=<0.5 10>, Min Scale Factor=<0.8 1.2>)
   &FN_Sprite_Attributes(
    Color=<rgba 0 0 0 1>
    Create Color Attribute=true
    Create SpriteSize Attribute=true
    SpriteSize=&Multiply_Vector2D_ByFloat(Float=1.5, Vector2D=$Particles.SpriteSize)
    Write to Module Namespace=true
   )
  ]
  rendererClasses [&NiagaraSpriteRendererProperties]
  renderers [
   {rendererClass=&NiagaraSpriteRendererProperties}
   {rendererClass=&NiagaraSpriteRendererProperties, rendererIndex=1}
  ]
 }
 Impact_Flare_and_Corona {
  emitterUpdateScript [
   &EmitterState(Life Cycle Mode=::NewEnumerator1, Loop Behavior=::NewEnumerator1, Loop Duration=0.5, UseLoopDelay=false)
   &SpawnBurst_Instantaneous(Spawn Count=1, Spawn Time=$System.OffSetAfterMotionBlur, Use Spawn Probability=false)
  ]
  particleSpawnScript [
   /Niagara/Modules/Spawn/Initialization/InitializeParticle(
    Color=<rgba 3 3 3 1>
    Lifetime=0.2
    Mesh Scale=&Multiply_VectorByFloat(Vector=&UniformRangedVector(Minimum=<1.3 1.3 0.5>, Maximum=<1.45 1.45 0.5>), Float=$System.scaleFX)
    Module.Write Color=true
    Module.Write Lifetime=true
    Module.Write Mass=true
    Module.Write Material Random=true
    Module.Write Position=false
    Module.Write Scale=true
    Module.Write SpriteRotation=false
    Module.Write SpriteSize=true
    Position=&SimulationPosition()
    Sprite Size=<200 175>
   )
   /SetVariables(Particles.CoronaSpriteSize=&Vector2DFromFloat(Value=&Multiply_Float(A=90, B=$System.scaleFX))) {isSetParameters=true}
   /SetVariables(Particles.ImpactColor:LinearColor=&MakeLinearColorFromVectorAndFloat("Vector (RGB)"=<1 1 1>, "Float (Alpha)"=&Multiply_Float(A=1, B=1))) {isSetParameters=true}
   &InitialMeshOrientation(
    Rotation=&RandomRangeVector(Minimum=<0 0 -1>, Maximum=<0 0 1>)
    Use Orientation Vector=true
    Use Rotation Vector=true
   )
   &CameraOffset(Camera Offset Amount=10)
  ]
  particleUpdateScript [
   &ParticleState()
   &DynamicMaterialParameters(
    Index 0 Param 1=&FloatFromCurve(FloatCurve=DataInterface() {propertyValues {Curve {keys [<at 0 1 RCIM_Cubic>,<at 0.8 0 RCIM_Cubic>]}}})
    Index 0 Param 2=2
    Index 0 Param 3=1
    Index 0 Param 4=1
    Index 1 Param 1=&FloatFromCurve(FloatCurve=DataInterface() {propertyValues {Curve {keys [
         {arriveTangent=-2.3163, interpMode=RCIM_Cubic, time=0, value=2}
         {arriveTangent=-1.9019, interpMode=RCIM_Cubic, leaveTangent=-1.9019, tangentMode=RCTM_User, time=1, value=0.4}
        ]}}})
    Index 1 Param 2=0.4
    Index 1 Param 3=0.6
    Index 1 Param 4=1
    Index 2 Param 1=&FloatFromCurve(FloatCurve=DataInterface() {propertyValues {Curve {keys [<at 0.46 0>,<at 1 1>]}}})
    Index 2 Param 2=45
    Module.Param0WriteEnabled=true
    Module.Param0WriteEnabled1=true
    Module.Param0WriteEnabled2=true
    Module.Param1WriteEnabled=true
    Module.Param1WriteEnabled1=true
    Module.Param1WriteEnabled2=true
    Module.Param2WriteEnabled=true
    Module.Param2WriteEnabled1=true
    Module.Param2WriteEnabled2=true
    Module.Param3WriteEnabled=true
    Module.Param3WriteEnabled1=true
    Module.Param3WriteEnabled2=true
    Write Parameter Index 1=true
    Write Parameter Index 2=true
   )
   /SetVariables(Particles.CoronaSpriteColor:LinearColor=&MakeLinearColorFromVectorAndFloat(
     "Vector (RGB)"=&MakeVectorFromLinearColorRGB(Color=<rgba 0.5521 0.5521 0.5521 1>)
     "Float (Alpha)"=&Multiply_Float(
      A=&FloatFromCurve(FloatCurve=DataInterface() {propertyValues {Curve {keys [<at 0.35 1.5 RCIM_Cubic>,<at 0.95 0>]}}})
      B=3.9
     )
    )) {isSetParameters=true}
   /SetVariables(Particles.ImpactColor:LinearColor=<rgba 1.6 1.12 0.64 1>) {isSetParameters=true}
   &ScaleColor(
    Scale Alpha=&FloatFromCurve(FloatCurve=DataInterface() {propertyValues {Curve {keys [<at 0 1>,<at 1 0>]}}})
    Scale RGB=&MakeVectorFromLinearColorRGB(Color=<rgba 1 1 1 1>)
    ScaleA=true
    ScaleRGB=true
   )
   &ScaleMeshSize(Scale Factor=&VectorFromFloat(Value=&FloatFromCurve(FloatCurve=DataInterface() {propertyValues {Curve {keys [<at 0 0 RCIM_Cubic>,<at 1 1.05>]}}})))
   /SetVariables(
    Particles.DY0:Vector4f=&MakeVector4(
     X=&FloatFromCurve(FloatCurve=DataInterface() {propertyValues {Curve {keys [{arriveTangent=1, time=0, value=-1}, <at 0.6 2>]}}})
     Y=0
     Z=0
     W=0
    )
    Particles.YellowFlashScale:Vector3f=&Multiply_VectorByFloat(
     Vector=$Particles.Initial.Scale
     Float=&FloatFromCurve(FloatCurve=DataInterface() {propertyValues {Curve {keys [<at 0 0.294 RCIM_Cubic>,<at 0.5 1.05>]}}})
    )
    Particles.float:NiagaraFloat=&FloatFromCurve(FloatCurve=DataInterface() {propertyValues {Curve {keys [<at 0 0.294 RCIM_Cubic>,<at 0.5 1.05>]}}})
   ) {isSetParameters=true}
  ]
  rendererClasses [&NiagaraSpriteRendererProperties, &NiagaraMeshRendererProperties]
  renderers [
   {rendererClass=&NiagaraSpriteRendererProperties}
   {rendererClass=&NiagaraMeshRendererProperties, rendererIndex=1}
  ]
 }
 Lazy_Sparks {
  emitterUpdateScript [
   &EmitterState(Life Cycle Mode=::NewEnumerator1, Loop Behavior=::NewEnumerator1, Loop Delay=0.05, Loop Duration=1, UseLoopDelay=true)
   &SpawnBurst_Instantaneous(
    Spawn Count=&UniformRangedInt(Minimum=15, Maximum=25)
    Spawn Time=$System.OffSetAfterMotionBlur
    Use Spawn Probability=false
   )
  ]
  particleSpawnScript [
   &InitializeParticle(
    Color=<rgba 3 3 3 1>
    Lifetime=&UniformRangedFloat(Minimum=0.1, Maximum=0.35)
    Module.Write Color=true
    Module.Write Lifetime=true
    Module.Write Mass=true
    Module.Write Material Random=true
    Module.Write Position=false
    Module.Write Scale=false
    Module.Write SpriteRotation=false
    Module.Write SpriteSize=true
    Position=&SimulationPosition()
    Sprite Size=&UniformRangedVector2D(Minimum=<1 1>, Maximum=<2 2>)
   )
   /Niagara/Modules/Spawn/Location/SphereLocation(
    MaskSpawnGroup=false
    Module.Override Randomness=false
    Sphere Origin:NiagaraPosition=/Niagara/DynamicInputs/Helpers/SimulationPosition()
    Sphere Radius=20
    Surface Only=false
   )
   &AddVelocity(Velocity=&Multiply_Vector(
     A=$Output.SphereLocation.SphereNormal
     B=&VectorFromFloat(Value=&UniformRangedFloat(Minimum=5, Maximum=20))
    )) {module=AddVelocity001}
   &AddVelocity(Velocity=&UniformRangedVector(Minimum=<-100 -100 150>, Maximum=<100 100 200>))
   &CameraOffset(Camera Offset Amount=25)
  ]
  particleUpdateScript [
   &ParticleState()
   &GravityForce()
   &AccelerationForce(Acceleration=<0 0 -500>)
   &SolveForcesAndVelocity(Clamp Velocity=false, Limit Acceleration=false)
   &ScaleSpriteSizeBySpeed(Max Scale Factor=<1 3>)
   &ScaleSpriteSize(Scale Factor=&Vector2DFromFloat(Value=&FloatFromCurve(FloatCurve=DataInterface() {propertyValues {Curve {keys [{arriveTangent=-6.725, interpMode=RCIM_Cubic, leaveTangent=-6.725, tangentMode=RCTM_User, time=0.65, value=1}, <at 1 0 RCIM_Cubic>]}}})))
   &ScaleColor(
    Scale RGB=&MakeVectorFromLinearColorRGB(Color=&ColorFromCurve(ColorCurve=DataInterface() {propertyValues {
        AlphaCurve {keys [<at 0 1 RCIM_Cubic>,<at 1 1 RCIM_Cubic>]}
        BlueCurve {keys [<at 0.2 0.6127 RCIM_Cubic>,<at 0.7 0.045 RCIM_Cubic>]}
        GreenCurve {keys [<at 0.2 3.4085 RCIM_Cubic>,<at 0.7 0.045 RCIM_Cubic>]}
        RedCurve {keys [<at 0.2 20 RCIM_Cubic>,<at 0.7 0.5 RCIM_Cubic>]}
       }}))
    ScaleA=true
    ScaleRGB=true
   )
  ]
  rendererClasses [&NiagaraSpriteRendererProperties]
  renderers [{rendererClass=&NiagaraSpriteRendererProperties}]
 }
 MIN_Flash001 {
  emitterUpdateScript [
   &EmitterState(Life Cycle Mode=::NewEnumerator1, Loop Behavior=::NewEnumerator1, Loop Duration=0.06, UseLoopDelay=true)
   &SpawnBurst_Instantaneous(Spawn Count=8, Spawn Time=$System.OffSetAfterMotionBlur, Use Spawn Probability=false)
  ]
  particleSpawnScript [
   &InitializeParticle(
    Lifetime=&RandomRangeFloat(Minimum=0.1, Maximum=0.13, Evaluation Type=::NewEnumerator1)
    Module.Write Color=true
    Module.Write Lifetime=true
    Module.Write Mass=true
    Module.Write Material Random=true
    Module.Write Position=false
    Module.Write Scale=false
    Module.Write SpriteRotation=false
    Module.Write SpriteSize=true
    Sprite Size=&RandomRangeVector2D(Minimum=<4 25>, Maximum=<6 35>, Evaluation Type=::NewEnumerator1)
   )
   /Niagara/Modules/Spawn/Location/V2/SphereLocation(
    MaskSpawnGroup=false
    Sphere Distribution:ENiagaraSphereDistributionMode=::NewEnumerator1
    Sphere Radius=7.5
    U Position:NiagaraFloat=&SelectFloatFromArray(
     Direct Array Index=&ReturnExecIndex()
     Direct Array Mode=::NewEnumerator1
     Float Selection Array=DataInterface() {propertyValues {FloatData [0.36,0.64,0.67,0.3586,0.34,0.67]}}
    )
    UseNonUniformScale=true
    UseOffset=false
    UseSphereOrientation=false
    V Position:NiagaraFloat=&SelectFloatFromArray(
     Direct Array Index=&ReturnExecIndex()
     Direct Array Mode=::NewEnumerator1
     Float Selection Array=DataInterface() {propertyValues {FloatData [0.32,0.36,0.33,0.6885,0.6,0.67]}}
    )
   ) {module=SphereLocation001}
   &CameraOffset(Camera Offset Amount=15)
   &AddVelocity(Velocity=&Multiply_VectorByFloat(Vector=$Output.SphereLocation001.SphereNormal, Float=1))
  ]
  particleUpdateScript [
   &ParticleState()
   &SolveForcesAndVelocity(Clamp Velocity=false, Limit Acceleration=false)
   &Color(
    Color=<rgba 2.5 0.75 0 1>
    Scale Alpha=&FloatFromCurve(FloatCurve=DataInterface() {propertyValues {Curve {keys [<at -0.0046 1.0113 RCIM_Cubic>,<at 1.0011 0.002>]}}})
   )
   &ScaleSpriteSize(Scale Factor=&Vector2DFromCurve(Vector2Curve=DataInterface() {propertyValues {XCurve {keys [<at 0 0.6>,<at 1 2>]}, YCurve {keys [<at 0 0.6>,<at 1 2>]}}}, ScaleCurve=<1 1>))
  ]
  rendererClasses [&NiagaraSpriteRendererProperties]
  renderers [{rendererClass=&NiagaraSpriteRendererProperties}]
 }
 SpeedLine {
  emitterUpdateScript [
   &EmitterState(Life Cycle Mode=::NewEnumerator1, Loop Behavior=::NewEnumerator1, Loop Duration=1, UseLoopDelay=false)
   &SpawnBurst_Instantaneous(Spawn Count=1, Spawn Time=$System.OffSetAfterMotionBlur, Use Spawn Probability=true)
   &SpawnBurst_Instantaneous(Spawn Count=1, Spawn Group=1, Spawn Time=$System.OffSetAfterMotionBlur, Use Spawn Probability=true) {module=SpawnBurst_Instantaneous001}
   &SpawnBurst_Instantaneous(Spawn Count=8, Spawn Group=2, Spawn Time=$System.OffSetAfterMotionBlur, Use Spawn Probability=true) {module=SpawnBurst_Instantaneous002}
  ]
  particleSpawnScript [
   &InitializeParticle(
    Color=<rgba 1 1 1 8>
    Lifetime=0.3
    Module.Write Color=true
    Module.Write Lifetime=true
    Module.Write Mass=true
    Module.Write Material Random=true
    Module.Write Position=false
    Module.Write Scale=false
    Module.Write SpriteRotation=true
    Module.Write SpriteSize=true
    Position=/Niagara/Functions/Localspace/SimulationPosition()
    Sprite Rotation=&RandomRangeFloat(Minimum=-360, Maximum=360)
    Sprite Size=&Vector2DFromFloat(Value=&Multiply_Float(A=165, B=&RandomRangeFloat(Minimum=0.7, Maximum=1.2)))
   ) {enabled=false}
   &PartitionParticles(
    Number Of Partitions=3
    Particles Per Partition:NiagaraInt32=&SelectIntFromArray(Direct Array Index=&SpawnGroup(), Int Selection Array=DataInterface() {propertyValues {IntData [1,1,10]}})
    Partition Mode:ENiagara_PartitionMode=::NewEnumerator1
   ) {enabled=false}
   /SetVariables(Particles.VisibilityTag=&SpawnGroup()) {isSetParameters=true}
   /Niagara/Modules/Spawn/Initialization/V2/InitializeParticle(
    Color=&MergeHelper_ColorSwitch(
     0=<rgba 1 0.46 0.4 8>
     1=<rgba 1 1 1 5>
     2=&RandomRangeLinearColor(Minimum=<rgba 0 0 0 0.7>, Maximum=<rgba 0 0 0 1>)
     3=?
     Index=$Particles.VisibilityTag
     Num=::NewEnumerator1
    )
    Lifetime=0.3
    Mass Mode=::NewEnumerator1
    Mesh Scale Mode=::NewEnumerator3
    Mesh Uniform Scale=&MakeFloatFromVector(
     Vector=&VectorFromFloat(Value=&Multiply_Float(
       A=$System.scaleFX
       B=&Multiply_Float(A=2.5, B=&RandomRangeFloat(Minimum=1, Maximum=1.8))
      ))
     Channel=::NewEnumerator0
    )
    Sprite Size=&MergeHelper_Vector2Switch(
     0=&Multiply_Vector2D(
      A=<1 0.5>
      B=&Vector2DFromFloat(Value=&Multiply_Float(
        A=$System.scaleFX
        B=&Multiply_Float(A=165, B=&RandomRangeFloat(Minimum=0.7, Maximum=1.2))
       ))
     )
     1=<0 0>
     2=&Multiply_Vector2D_ByFloat(Float=$System.scaleFX, Vector2D=&RandomRangeVector2D(Minimum=<10 44.8>, Maximum=<5 33.6>))
     3=?
     4=?
     Index=$Particles.VisibilityTag
     Num=::NewEnumerator1
    )
    Sprite Size Mode=::NewEnumerator4
    UsePositionOffset=false
    Write Material Random=true
   ) {module=InitializeParticle001}
   &SphereLocation(
    MaskSpawnGroup=false
    Sphere Distribution:ENiagaraSphereDistributionMode=::NewEnumerator1
    Sphere Radius=1
    U Position:NiagaraFloat=&RandomRangeFloat(Minimum=0, Maximum=1)
    UseNonUniformScale=true
    UseOffset=false
    UseSphereOrientation=false
    V Position:NiagaraFloat=&RandomRangeFloat(Minimum=0.3, Maximum=0.7)
   ) {module=SphereLocation001}
   &AddVelocity(Velocity=&Multiply_VectorByFloat(
     Vector=$Output.SphereLocation001.SphereNormal
     Float=&MergeHelper_FloatSwitch(0=0, 1=0, 2=140, 3=?, Index=$Particles.VisibilityTag, Num=::NewEnumerator1)
    ))
   &CameraOffset(Camera Offset Amount=5)
  ]
  particleUpdateScript [
   &ParticleState()
   &SolveForcesAndVelocity(Clamp Velocity=false, Limit Acceleration=false)
   &ScaleSpriteSize(Scale Factor=&MergeHelper_Vector2Switch(
     0=&Vector2DFromFloat(Value=&FloatFromCurve(
       FloatCurve=DataInterface() {propertyValues {Curve {keys [{arriveTangent=3.9086, interpMode=RCIM_Cubic, leaveTangent=3.9086, tangentMode=RCTM_User, time=0, value=0}, <at 1 1 RCIM_Cubic>]}}}
       Scale Curve=0.8
      ))
     1=<0 0>
     2=&Vector2DFromCurve(
      Vector2Curve=DataInterface() {propertyValues {XCurve {keys [<at 0 0.778>,<at 1 1>]}, YCurve {keys [<at 0 0.03>,<at 1 1.5>]}}}
      ScaleCurve=<1 1>
     )
     3=<0 0>
     4=<0 0>
     Index=0
     Num=::NewEnumerator3
    ))
   &DynamicMaterialParameters(
    Index 0 Param 1=&FloatFromCurve(FloatCurve=DataInterface() {propertyValues {Curve {keys [<at 0 1 RCIM_Cubic>, {arriveTangent=-2.742, interpMode=RCIM_Cubic, leaveTangent=-2.742, tangentMode=RCTM_User, time=1, value=0}]}}})
    Index 0 Param 2=&FloatFromCurve(FloatCurve=DataInterface() {propertyValues {Curve {keys [<at 0 0 RCIM_Cubic>, {arriveTangent=2.4833, interpMode=RCIM_Cubic, leaveTangent=2.4833, tangentMode=RCTM_User, time=1, value=1}]}}})
    Module.Param0WriteEnabled=true
    Module.Param1WriteEnabled=true
    Module.Param2WriteEnabled=true
    Module.Param3WriteEnabled=true
   )
   &ScaleColor(
    Scale Alpha=&MergeHelper_FloatSwitch(
     0=&FloatFromCurve(FloatCurve=DataInterface() {propertyValues {Curve {keys [
          {arriveTangent=-0.2921, interpMode=RCIM_Cubic, leaveTangent=-0.2921, tangentMode=RCTM_User, time=0, value=7}
          {arriveTangent=-7, interpMode=RCIM_Cubic, leaveTangent=-7, time=0.4, value=4}
          <at 1 0 RCIM_Cubic>
         ]}}})
     1=0
     2=&FloatFromCurve(FloatCurve=DataInterface() {propertyValues {Curve {keys [<at 0 10>,<at 0.7 0>]}}}, Scale Curve=0.2)
     3=?
     Index=$Particles.VisibilityTag
     Num=::NewEnumerator1
    )
    ScaleA=true
    ScaleRGB=false
   )
   /SetVariables(
    Particles.DY0:Vector4f=&MakeVector4(X=1, Y=&RandomRangeFloat(Minimum=1.5, Maximum=2), Z=1, W=0)
    Particles.FlashColor:LinearColor=&MultiplyLinearColor(
     Linear Color=$Particles.Initial.Color
     Scale Factor=&MakeFloat4FromVectorAndFloat(
      "Vector (RGB)"=<1 1 1>
      "Float (Alpha)"=&FloatFromCurve(FloatCurve=DataInterface() {propertyValues {Curve {keys [
           {arriveTangent=44.6077, interpMode=RCIM_Cubic, leaveTangent=44.6077, tangentMode=RCTM_User, time=0, value=0}
           <at 0.0781 1.0138 RCIM_Cubic>
           {arriveTangent=-1.8607, interpMode=RCIM_Cubic, leaveTangent=-1.8607, tangentMode=RCTM_User, time=1, value=0}
          ]}}})
     )
    )
   ) {isSetParameters=true}
   &ScaleMeshSize(Scale Factor=&VectorFromFloat(Value=&FloatFromCurve(FloatCurve=DataInterface() {propertyValues {Curve {keys [{arriveTangent=2.9993, interpMode=RCIM_Cubic, leaveTangent=2.9993, tangentMode=RCTM_User, time=0, value=0}, <at 1 1 RCIM_Cubic>]}}})))
  ]
  rendererClasses [&NiagaraSpriteRendererProperties, &NiagaraMeshRendererProperties]
  renderers [
   {rendererClass=&NiagaraSpriteRendererProperties}
   {rendererClass=&NiagaraMeshRendererProperties, rendererIndex=1}
   {rendererClass=&NiagaraSpriteRendererProperties, rendererIndex=2}
  ]
 }
 Toon_Smoke {
  emitterUpdateScript [
   &EmitterState(Life Cycle Mode=::NewEnumerator1, Loop Behavior=::NewEnumerator1, Loop Delay=0.02, Loop Duration=2, UseLoopDelay=true)
   &SpawnBurst_Instantaneous(Spawn Count=2, Spawn Time=$System.OffSetAfterMotionBlur, Use Spawn Probability=false)
  ]
  particleSpawnScript [
   /SetVariables(Particles.Random=&RandomRangeFloat(Minimum=0, Maximum=1)) {isSetParameters=true}
   &PartitionParticles(Particles Per Partition=1, Partition Mode:ENiagara_PartitionMode=::NewEnumerator1)
   &InitializeParticle(
    Color=&LerpLinearColors(StartColor=<rgba 1 1 1 1>, EndColor=$System.TODColor, LerpFactor=0.8)
    Lifetime=&RandomRangeFloat(Minimum=0.4, Maximum=0.5)
    Sprite Rotation Mode=::NewEnumerator2
    Sprite Size Max=<120 160>
    Sprite Size Min=<100 140>
    Sprite Size Mode=::NewEnumerator2
    Sprite UV Mode=::NewEnumerator1
    UsePositionOffset=false
    Write Material Random=true
   ) {module=InitializeParticle001}
   &ShapeLocation(
    MaskSpawnGroup=false
    Rotation Mode=::NewEnumerator2
    Sphere Distribution=::NewEnumerator3
    Sphere Radius=5
    UseOwnerScale=true
    Yaw / Pitch / Roll=<0.5 0 0>
   )
   &CameraOffset(Camera Offset Amount=30)
   &AddVelocity(
    Module.Use Falloff Distance=false
    Origin Offset=<-5 0 0>
    Override Seed=false
    Velocity Mode=::NewEnumerator1
    Velocity Speed=&RandomRangeFloat(Minimum=25, Maximum=50)
   )
   &DynamicMaterialParameters(
    Index 0 Param 2=&UniformAOrBFloat(A=0)
    Module.Param0WriteEnabled=true
    Module.Param1WriteEnabled=true
    Module.Param2WriteEnabled=true
    Module.Param3WriteEnabled=true
   ) {module=DynamicMaterialParameters001}
  ]
  particleUpdateScript [
   &ParticleState()
   &ScaleSpriteSize(Scale Sprite Size Mode=::NewEnumerator0, Uniform Scale Factor=0.7)
   &Drag(Drag=2, Use Linear Drag=true, Use Rotational Drag=true)
   &SolveForcesAndVelocity(Clamp Velocity=false, Limit Acceleration=false)
   &DynamicMaterialParameters(
    Index 0 Param 1=&FloatFromCurve(FloatCurve=DataInterface() {propertyValues {Curve {keys [
         {arriveTangent=1.3143, interpMode=RCIM_Cubic, leaveTangent=1.3143, tangentMode=RCTM_User, time=0, value=0}
         {arriveTangent=0.4262, interpMode=RCIM_Cubic, leaveTangent=0.4262, tangentMode=RCTM_User, time=1, value=1}
        ]}}})
    Module.Param0WriteEnabled=true
    Module.Param1WriteEnabled=false
    Module.Param2WriteEnabled=true
    Module.Param3WriteEnabled=true
   )
   &ScaleColor(
    Scale Alpha=&FloatFromCurve(
     FloatCurve=DataInterface() {propertyValues {Curve {keys [{interpMode=RCIM_Cubic, tangentMode=RCTM_User, time=0, value=1}, {interpMode=RCIM_Cubic, tangentMode=RCTM_User, time=1, value=0}]}}}
     Scale Curve=0.5
    )
    Scale RGB=&VectorFromFloat(Value=&FNTimeOfDayScalar(Night Scale=0.6, Morning Scale=0.75, Evening Scale=0.75))
    ScaleA=true
    ScaleRGB=true
   )
   &Color(Color=&ColorFromCurve(ColorCurve=DataInterface() {propertyValues {
       AlphaCurve {keys [<at 0 1>,<at 0.8091 0.5>,<at 1 0>]}
       BlueCurve {keys [<at 0.0609 0.2075 RCIM_Cubic>,<at 0.4636 3 RCIM_Cubic>]}
       GreenCurve {keys [<at 0.0609 2.4438 RCIM_Cubic>,<at 0.4636 3 RCIM_Cubic>]}
       RedCurve {keys [<at 0.0609 6.9163 RCIM_Cubic>,<at 0.4636 3 RCIM_Cubic>]}
      }})) {enabled=false}
  ]
  rendererClasses [&NiagaraSpriteRendererProperties]
  renderers [{rendererClass=&NiagaraSpriteRendererProperties}]
 }
}
summary {
 systemName=NS_Emote_StylusFluff_Equip2
 emitters [
  {
   bEnabled=false
   emitterName=Flash_Flare
   rendererClasses [&NiagaraSpriteRendererProperties, &NiagaraLightRendererProperties]
  }
  {emitterName=Lazy_Sparks, rendererClasses [&NiagaraSpriteRendererProperties]}
  {emitterName=MIN_Flash001, rendererClasses [&NiagaraSpriteRendererProperties]}
  {emitterName=Toon_Smoke, rendererClasses [&NiagaraSpriteRendererProperties]}
  {emitterName=SpeedLine, rendererClasses [&NiagaraSpriteRendererProperties, &NiagaraMeshRendererProperties]}
  {
   emitterName=Impact_Flare_and_Corona
   rendererClasses [&NiagaraSpriteRendererProperties, &NiagaraMeshRendererProperties]
  }
  {bEnabled=false, emitterName=ComicLinesImpact, rendererClasses [&NiagaraSpriteRendererProperties]}
  {emitterName=GlowDrops, rendererClasses [&NiagaraSpriteRendererProperties]}
  {bEnabled=false, emitterName=ComicLinesAxeMotionBlur, rendererClasses [&NiagaraSpriteRendererProperties]}
 ]
 userVariables [
  User.bIsFrontEnd {type=NiagaraBool}
  User.bIsFrontEndPreview {type=NiagaraBool}
  User.PhysicalSurface {type=NiagaraInt32}
  User.SecondImpact {type=NiagaraBool}
  User.SwingSectionIndex {type=NiagaraInt32}
  User.WeakpointHit {type=NiagaraBool}
 ]
}
